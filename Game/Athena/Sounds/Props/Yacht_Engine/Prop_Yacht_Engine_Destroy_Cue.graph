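# Sound cue for the destruction of the yacht engine prop.
# A random node picks one of three equally weighted destroy wave variants and is excluded from branch culling.
# It uses the impacts sound class with demolition attenuation for medium-sized objects and a slight volume boost.
# A concurrency override caps simultaneous instances at a small count, stopping the lowest priority sound first.
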SoundCue /Game/Athena/Sounds/Props/Yacht_Engine/Prop_Yacht_Engine_Destroy_Cue
root=Random_2
nodes [
 {id=Random_2, children [WavePlayer_8, WavePlayer_9, WavePlayer_10], properties {bShouldExcludeFromBranchCulling=true, weights [1,1,1]}}
 {id=WavePlayer_10, properties {soundWaveAssetPtr=&/prop_yacht_engine_destroy_03}}
 {id=WavePlayer_8, properties {soundWaveAssetPtr=&/prop_yacht_engine_destroy_01}}
 {id=WavePlayer_9, properties {soundWaveAssetPtr=&/prop_yacht_engine_destroy_02}}
]
properties {
 attenuationSettings=/Game/Sounds/Attenuation/Fort_Demolition_Destroy_Medium_Sized_Object_Att
 bOverrideConcurrency=true
 maxDistance=4000
 soundClassObject=&Fort_Impacts
 volumeMultiplier=1.2
 concurrencyOverrides {maxCount=2, resolutionRule=StopLowestPriority}
}
