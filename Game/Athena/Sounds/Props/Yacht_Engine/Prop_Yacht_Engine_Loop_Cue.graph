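# Sound cue for the yacht engine prop's ambient loop in Athena, with attenuation and limited concurrency.
# A mixer blends a randomly chosen, pitch-varied engine loop, faded in by an enveloper, with a slightly louder layer of occasional engine noises.
# That layer repeatedly waits several seconds, then plays a random steam, wronk or clank one-shot with varied pitch and volume.
# A quality node keeps these one-shots in its first two levels only.
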SoundCue /Game/Athena/Sounds/Props/Yacht_Engine/Prop_Yacht_Engine_Loop_Cue
root=Mixer_0
nodes [
 {id=Delay_0, children [Random_1], properties {delayMax=10, delayMin=6}}
 {
  id=Enveloper_0
  children [Modulator_0]
  properties {pitchCurve {editorCurveData {keys [{time=0}]}}, volumeCurve {editorCurveData {keys [<at 0 0>,<at 0.5 1>]}}}
 }
 {id=Looping_0, children [Delay_0]}
 {id=Mixer_0, children [Enveloper_0, QualityLevel_0], properties {inputVolume [1,1.5]}}
 {id=Modulator_0, children [Random_0], properties {pitchMax=1.3, pitchMin=0.9, volumeMax=1, volumeMin=1}}
 {id=Modulator_1, children [Looping_0], properties {pitchMax=1.2, pitchMin=0.8, volumeMax=1, volumeMin=0.3}}
 {id=QualityLevel_0, children [Modulator_1,Modulator_1,null,null]}
 {id=Random_0, children [WavePlayer_0,WavePlayer_1,WavePlayer_7], properties {bShouldExcludeFromBranchCulling=true, weights [1,1,1]}}
 {
  id=Random_1
  children [WavePlayer_2,WavePlayer_3,WavePlayer_4,WavePlayer_5,WavePlayer_6]
  properties {bShouldExcludeFromBranchCulling=true, weights [1,1,1,1,1]}
 }
 {id=WavePlayer_0, properties {bLooping=true, soundWaveAssetPtr=&/prop_yacht_engine_loop_01}}
 {id=WavePlayer_1, properties {bLooping=true, soundWaveAssetPtr=&/prop_yacht_engine_loop_02}}
 {id=WavePlayer_2, properties {soundWaveAssetPtr=&/prop_yacht_engine_steam_01}}
 {id=WavePlayer_3, properties {soundWaveAssetPtr=&/prop_yacht_engine_steam_02}}
 {id=WavePlayer_4, properties {soundWaveAssetPtr=&/prop_yacht_engine_steam_03}}
 {id=WavePlayer_5, properties {soundWaveAssetPtr=&/prop_yacht_engine_wronk_01}}
 {id=WavePlayer_6, properties {soundWaveAssetPtr=&/prop_yacht_engine_clank_01}}
 {id=WavePlayer_7, properties {bLooping=true, soundWaveAssetPtr=&/prop_yacht_engine_loop_03}}
]
properties {
 attenuationSettings=&/Prop_Yacht_Engine_Att
 bOverrideConcurrency=true
 maxDistance=1500
 priority=100
 soundClassObject=&Fort_Ambient_Props_Athena
 volumeMultiplier=2.005
 concurrencyOverrides {maxCount=3, voiceStealReleaseTime=1.5, volumeScale=0.75}
}
